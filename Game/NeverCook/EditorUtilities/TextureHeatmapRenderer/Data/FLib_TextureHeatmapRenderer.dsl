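; Editor utility function library that renders heatmap data objects into 2D render targets as textures.
; GenerateHeatmapTexture checks the data object, render target and bounds, printing a message on failure, and calls DrawHeatmap.
; DrawHeatmap clears the target, maps each location's XY from the data bounds into pixel space, and draws a centred dot texture there.
; Dots are tinted by sampling a color curve with normalized severity and blended additively, so overlapping points accumulate.
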(fn DrawHeatmap
  (params (data :type Heatmap%20Renderer%20Data%20Object%20Object%20Reference) (renderTarget :type Texture%20Render%20Target%202D%20Object%20Reference) (dotSize :type Float%20%28double-precision%29 :default 32.0))
  (bind (_bounds_min _bounds_max) (Utilities|Struct|BreakBox2D (Variables|/Game/Blueprints/HeatmapRenderer/Data/HeatmapRendererDataObject.HeatmapRendererDataObject_C|GetBounds_Min (LocalVariables|Getl_data))))
  (bind (_x _y) (Math|Vector2D|BreakVector2D _bounds_max))
  (bind (_x_2 _y_2) (Math|Vector2D|BreakVector2D _bounds_min))
  (bind _l_data (LocalVariables|Getl_data))
  (LocalVariables|Setl_data data)
  (LocalVariables|Setl_renderTarget renderTarget)
  (LocalVariables|Setl_size dotSize)
  (Utilities|FlowControl|Sequence
    (:then_0
      (Rendering|ClearRenderTarget2D 0 (LocalVariables|Getl_renderTarget))
      (bind (_canvas _size _context) (Rendering|BeginDrawCanvastoRenderTarget 0 (LocalVariables|Getl_renderTarget)))
      (LocalVariables|Setl_canvas _canvas)
      (LocalVariables|Setl_rtSize _size)
      (LocalVariables|Setl_context _context))
    (:then_1
      (bind _normalized (Class|HeatmapRendererDataObject|GetSeveritiesNormalized _l_data))
      (for (_array_element _array_index) (Variables|/Game/Blueprints/HeatmapRenderer/Data/HeatmapRendererDataObject.HeatmapRendererDataObject_C|GetLocations _l_data)
        (Utilities|FlowControl|Sequence
          (:then_0
            (LocalVariables|Setl_drawLoc (* (Math|Vector2D|MakeVector2D (Math|Float|NormalizetoRange (.x _array_element) _x_2 _x) (Math|Float|NormalizetoRange (.y _array_element) _y_2 _y)) (LocalVariables|Getl_rtSize))))
          (:then_1
            (Canvas|DrawTexture :self (LocalVariables|Getl_canvas) :RenderTexture (LocalVariables|Getl_textureToDraw) :ScreenPosition (- (LocalVariables|Getl_drawLoc) (/ (LocalVariables|Getl_size) 2.0)) :ScreenSize (Math|Conversions|ToVector2D (LocalVariables|Getl_size)) :CoordinateSize "(X=1.000000,Y=1.000000)" :RenderColor (Math|Curves|GetLinearColorValue (LocalVariables|Getl_colorCurve) (Utilities|Array|Get(aref) _normalized _array_index)) :BlendMode "BLEND_Additive" :Rotation 0.0 :PivotPoint "(X=0.500000,Y=0.500000)")))))
    (:then_2
      (Rendering|EndDrawCanvastoRenderTarget 0 (LocalVariables|Getl_context)))))

(fn GenerateHeatmapTexture
  (params (heatmapDataObject :type Heatmap%20Renderer%20Data%20Object%20Object%20Reference) (renderTarget :type Texture%20Render%20Target%202D%20Object%20Reference) (dotSize :type Float%20%28double-precision%29 :default 32.0))
  (bind _l_data (LocalVariables|Getl_data))
  (bind _bounds (Variables|/Game/Blueprints/HeatmapRenderer/Data/HeatmapRendererDataObject.HeatmapRendererDataObject_C|GetBounds _l_data))
  (bind (_min _max _bisvalid) (Utilities|Struct|BreakBox2D _bounds))
  (LocalVariables|Setl_data heatmapDataObject)
  (LocalVariables|Setl_rt renderTarget)
  (Utilities|FlowControl|Sequence
    (:then_0
      (Utilities|IsValid (LocalVariables|Getl_data)
        (:Is%20Valid
          (Utilities|IsValid (LocalVariables|Getl_rt)
            (:Is%20Valid
              (if (not _bisvalid)
                (Development|PrintString 0 (Utilities|String|Append "Box Invalid: " (Utilities|String|ToString(Vector2d) _min) ", " (Utilities|String|ToString(Vector2d) _max)))))
            (:Is%20Not%20Valid
              (Development|PrintString 0 "No RT"))))
        (:Is%20Not%20Valid
          (Development|PrintString 0 "No Data")))
      (return false))
    (:then_1
      (CallFunction|DrawHeatmap (LocalVariables|Getl_data) (LocalVariables|Getl_rt) dotSize))))
